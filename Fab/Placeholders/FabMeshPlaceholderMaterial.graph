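# Translucent placeholder material for Fab meshes that shows the Fab logo as a progress indicator, with depth testing disabled.
# AlignMeshToTheCamera provides billboard-style alignment toward the camera.
# The logo texture's alpha is multiplied by a hard horizontal wipe, made by flooring U plus one minus the Progress parameter.
# The portion below Progress keeps full alpha and the remainder drops to half.
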Material /Fab/Placeholders/FabMeshPlaceholderMaterial
graph {
 properties {
  bAllowFrontLayerTranslucency=false
  bCastRayTracedShadows=false
  bDisableDepthTest=true
  bFullyRough=true
  bTangentSpaceNormal=false
  bUseLightmapDirectionality=false
  bUseTranslucencyVertexFog=false
  bWriteOnlyAlpha=true
  blendMode=BLEND_Translucent
  materialDecalResponse=MDR_None
  pixelDepthOffsetMode=PDOM_Legacy
  previewMesh=&/MeshPlaceholder
  translucencyLightingMode=TLM_VolumetricPerVertexNonDirectional
  usedShadingModels=Default Lit
 }
 nodes [
  MaterialFunctionCall!(
   WorldPosition
   PivotLocation
   WS Normals
   Custom Object Basis 1
   Custom Object Basis 2=Constant3Vector() {constant=<rgba 1 0 0 0>}
   Custom Object Basis 3=Constant3Vector() {constant=<rgba 0 1 0 0>}
  ) {materialFunction=/Engine/Functions/Engine_MaterialFunctions02/WorldPositionOffset/AlignMeshToTheCamera}
  n1 = TextureCoordinate()
  n2 = Add(
   A=MaterialFunctionCall(Float2=@n1) {materialFunction=&BreakOutFloat2Components}.R
   B=OneMinus(None=ScalarParameter() {defaultValue=1, parameterName=Progress, sliderMax=1})
  )
  Multiply(A=TextureSample(UVs=@n1) {texture=&/FabLogo}.A, B=LinearInterpolate(Alpha=Floor(None=@n2)) {constA=1, constB=0.5})
 ]
}
